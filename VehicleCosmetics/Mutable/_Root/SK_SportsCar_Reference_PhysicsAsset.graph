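# Physics asset for the SK_SportsCar_Reference skeletal mesh in the Mutable root of the vehicle cosmetics content.
# A single body on the root bone approximates the whole car with one large capsule, offset upward and rotated away from the bone's default axes.
# It has no constraints and no bodies on any other bone, so it provides one coarse volume with no articulated parts.
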PhysicsAsset /VehicleCosmetics/Mutable/_Root/SK_SportsCar_Reference_PhysicsAsset
bodies {root [{shapeType=Capsule, length=254.6544, radius=133.1865, center=<15.9434 -1.9349 66.3621>, rotation=<rot 0 -91.1128 -86.4346>}]}
constraints []
