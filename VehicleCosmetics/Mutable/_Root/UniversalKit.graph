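# UniversalKit is a Default Lit material in the Mutable root folder of the vehicle cosmetics content.
# It enables usage with skeletal meshes and mesh deformers.
# The graph holds a single vector parameter with a pale grey-green default, and shows no connection from it to a material output.
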Material /VehicleCosmetics/Mutable/_Root/UniversalKit
graph {
 properties {bUsedWithMeshDeformer=true, bUsedWithSkeletalMesh=true, pixelDepthOffsetMode=PDOM_Legacy, usedShadingModels=Default Lit}
 nodes [VectorParameter() {defaultValue=<rgba 0.5911 0.6368 0.6191 0>}]
}
